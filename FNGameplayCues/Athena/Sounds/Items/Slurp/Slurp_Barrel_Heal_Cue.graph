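# Sound cue for the Slurp Barrel heal effect in Athena, part of the gameplay cue sounds.
# A random node picks one of four heal wave variations with equal weighting.
# It uses the consumable-use attenuation and the consumables sound class, with a boosted volume.
# A concurrency override caps simultaneous instances and stops the oldest when the limit is exceeded.
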SoundCue /FNGameplayCues/Athena/Sounds/Items/Slurp/Slurp_Barrel_Heal_Cue
root=Random_0
nodes [
 {id=Random_0, children [WavePlayer_1,WavePlayer_2,WavePlayer_3,WavePlayer_4], properties {weights [1,1,1,1]}}
 {id=WavePlayer_1, properties {soundWaveAssetPtr=&/Item_Slurp_Barrel_Heal_01}}
 {id=WavePlayer_2, properties {soundWaveAssetPtr=&/Item_Slurp_Barrel_Heal_02}}
 {id=WavePlayer_3, properties {soundWaveAssetPtr=&/Item_Slurp_Barrel_Heal_03}}
 {id=WavePlayer_4, properties {soundWaveAssetPtr=&/Item_Slurp_Barrel_Heal_04}}
]
properties {
 attenuationSettings=&Fort_Consumable_Use_Athena_Att
 bOverrideConcurrency=true
 maxDistance=2250
 soundClassObject=&Fort_Consumables
 volumeMultiplier=1.5
 concurrencyOverrides {maxCount=4, resolutionRule=StopOldest}
}
