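# Slurp drain-heal loop sound cue for Fortnite Athena, using consumable attenuation and sound class.
# A mixer layers a looping drain-heal wave, which an enveloper fades in from silence over two seconds, with a non-looping barrel heal wave at full input volume.
# A concurrency override caps simultaneous instances at a small count and stops the oldest when exceeded.
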SoundCue /FNGameplayCues/Athena/Sounds/Items/Slurp/Slurp_Drain_Heal_Loop_Cue
root=Mixer_1
nodes [
 {
  id=Enveloper_1
  children [WavePlayer_0]
  properties {pitchCurve {editorCurveData {keys [{time=0}]}}, volumeCurve {editorCurveData {keys [<at 0 0>,<at 2 1>]}}}
 }
 {id=Mixer_1, children [Enveloper_1,WavePlayer_3], properties {inputVolume [0.8,1]}}
 {id=WavePlayer_0, properties {bLooping=true, soundWaveAssetPtr=/Game/Athena/Sounds/Items/Slurp/Item_Slurp_Drain_Heal_Loop_01}}
 {id=WavePlayer_3, properties {soundWaveAssetPtr=&/Item_Slurp_Barrel_Heal_04}}
]
properties {
 attenuationSettings=&Fort_Consumable_Use_Athena_Att
 bOverrideConcurrency=true
 maxDistance=2250
 soundClassObject=&Fort_Consumables
 volumeMultiplier=0.6
 concurrencyOverrides {maxCount=4, resolutionRule=StopOldest}
}
