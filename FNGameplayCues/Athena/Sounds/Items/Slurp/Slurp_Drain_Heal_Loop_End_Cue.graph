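# Sound cue for the end of the Slurp drain heal loop, using a Slurp barrel heal sound.
# A single wave player feeds the output directly, played at raised pitch and reduced volume through the consumables sound class and consumable-use attenuation.
# Overridden concurrency limits it to one instance, stopping the oldest, with a retrigger time set.
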SoundCue /FNGameplayCues/Athena/Sounds/Items/Slurp/Slurp_Drain_Heal_Loop_End_Cue
root=WavePlayer_3
nodes [{id=WavePlayer_3, properties {soundWaveAssetPtr=&/Item_Slurp_Barrel_Heal_04}}]
properties {
 attenuationSettings=&Fort_Consumable_Use_Athena_Att
 bOverrideConcurrency=true
 maxDistance=2250
 pitchMultiplier=2
 soundClassObject=&Fort_Consumables
 volumeMultiplier=0.4
 concurrencyOverrides {maxCount=1, resolutionRule=StopOldest, retriggerTime=3}
}
