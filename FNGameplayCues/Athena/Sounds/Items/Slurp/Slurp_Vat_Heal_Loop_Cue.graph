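# Sound cue for the Slurp vat healing loop in Athena, routed through the consumables sound class and use attenuation.
# A mixer layers a one-shot barrel heal wave with a looping vat heal wave, weighting the loop higher.
# An enveloper fades the loop in over its first couple of seconds while the one-shot plays immediately.
# Overridden concurrency caps simultaneous instances and stops the oldest.
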SoundCue /FNGameplayCues/Athena/Sounds/Items/Slurp/Slurp_Vat_Heal_Loop_Cue
root=Mixer_0
nodes [
 {
  id=Enveloper_0
  children [WavePlayer_1]
  properties {pitchCurve {editorCurveData {keys [{time=0}]}}, volumeCurve {editorCurveData {keys [<at 0 0>,<at 2 1>]}}}
 }
 {id=Mixer_0, children [WavePlayer_0,Enveloper_0], properties {inputVolume [1.5,2]}}
 {id=WavePlayer_0, properties {soundWaveAssetPtr=&/Item_Slurp_Barrel_Heal_04}}
 {id=WavePlayer_1, properties {bLooping=true, soundWaveAssetPtr=/Game/Athena/Sounds/Items/Slurp/Item_Slurp_Vat_Heal_Loop_02}}
]
properties {
 attenuationSettings=&Fort_Consumable_Use_Athena_Att
 bOverrideConcurrency=true
 maxDistance=2250
 soundClassObject=&Fort_Consumables
 volumeMultiplier=0.3
 concurrencyOverrides {maxCount=4, resolutionRule=StopOldest}
}
