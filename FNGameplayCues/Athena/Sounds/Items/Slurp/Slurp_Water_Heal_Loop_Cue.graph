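# Sound cue for the Athena Slurp water heal loop, built from layered Slurp item sounds.
# A mixer combines a one-shot barrel heal with two looping layers that enveloper volume curves fade in over a couple of seconds: a vat heal loop and a randomly chosen water ambience with slight downward pitch modulation.
# It uses consumable attenuation and sound class, with a concurrency override that stops the oldest instance.
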SoundCue /FNGameplayCues/Athena/Sounds/Items/Slurp/Slurp_Water_Heal_Loop_Cue
root=Mixer_0
nodes [
 {
  id=Enveloper_0
  children [WavePlayer_1]
  properties {pitchCurve {editorCurveData {keys [{time=0}]}}, volumeCurve {editorCurveData {keys [<at 0 0>,<at 2 1>]}}}
 }
 {
  id=Enveloper_1
  children [Random_0]
  properties {pitchCurve {editorCurveData {keys [{time=0}]}}, volumeCurve {editorCurveData {keys [<at 0 0>,<at 2.5 1>]}}}
 }
 {id=Mixer_0, children [WavePlayer_0,Enveloper_0,Modulator_0], properties {inputVolume [1.5,2,2.5]}}
 {id=Modulator_0, children [Enveloper_1], properties {pitchMax=1, pitchMin=0.9}}
 {id=Random_0, children [WavePlayer_2,WavePlayer_3], properties {weights [1,1]}}
 {id=WavePlayer_0, properties {soundWaveAssetPtr=&/Item_Slurp_Barrel_Heal_04}}
 {id=WavePlayer_1, properties {bLooping=true, soundWaveAssetPtr=/Game/Athena/Sounds/Items/Slurp/Item_Slurp_Vat_Heal_Loop_02}}
 {id=WavePlayer_2, properties {bLooping=true, soundWaveAssetPtr=/Game/Athena/Sounds/Items/Slurp/Item_Slurp_Water_Amb_Loop_Medium_01}}
 {id=WavePlayer_3, properties {bLooping=true, soundWaveAssetPtr=/Game/Athena/Sounds/Items/Slurp/Item_Slurp_Water_Amb_Loop_Medium_02}}
]
properties {
 attenuationSettings=&Fort_Consumable_Use_Athena_Att
 bOverrideConcurrency=true
 maxDistance=2250
 soundClassObject=&Fort_Consumables
 volumeMultiplier=0.5
 concurrencyOverrides {maxCount=4, resolutionRule=StopOldest}
}
